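# Sound cue for the Ashton Milo LTM projectile impact on a player, built from enemy impact waves.
# A Random node picks one of six equally weighted wave variations, then passes it through a Modulator at the root.
# It plays at reduced volume in the impacts sound class, using the shared Athena bullet-impact attenuation.
# A concurrency override prevents new instances once the limit is reached.
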SoundCue /Figment_Martin/Athena/Sounds/LTMs/Ashton/Milo/Ashton_Milo_Impact_Player_Cue
root=Modulator_0
nodes [
 {id=Modulator_0, children [Random_0]}
 {id=Random_0, children [WavePlayer_0,WavePlayer_1,WavePlayer_2,WavePlayer_3,WavePlayer_4,WavePlayer_5], properties {weights [1,1,1,1,1,1]}}
 {id=WavePlayer_0, properties {soundWaveAssetPtr=&/Ashton_Milo_BR_Proj_Enemy_Impact_06}}
 {id=WavePlayer_1, properties {soundWaveAssetPtr=&/Ashton_Milo_BR_Proj_Enemy_Impact_05}}
 {id=WavePlayer_2, properties {soundWaveAssetPtr=&/Ashton_Milo_BR_Proj_Enemy_Impact_04}}
 {id=WavePlayer_3, properties {soundWaveAssetPtr=&/Ashton_Milo_BR_Proj_Enemy_Impact_03}}
 {id=WavePlayer_4, properties {soundWaveAssetPtr=&/Ashton_Milo_BR_Proj_Enemy_Impact_02}}
 {id=WavePlayer_5, properties {soundWaveAssetPtr=&/Ashton_Milo_BR_Proj_Enemy_Impact_01}}
]
properties {
 attenuationSettings=/Game/Sounds/Attenuation/Fort_Bullet_Impact_Athena_Att
 bOverrideConcurrency=true
 maxDistance=4500
 soundClassObject=&Fort_Impacts
 volumeMultiplier=0.6
 attenuationOverrides {distanceAlgorithm=NaturalSound, falloffDistance=4000, attenuationShapeExtents {x=1000}}
 concurrencyOverrides {resolutionRule=PreventNew}
}
