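# Sound cue for the Milo pickup sound in the Ashton limited-time mode content.
# A single Wave Player feeds the Ashton_Milo_BR_Pickup wave straight to the output at reduced volume, with no randomization or mixing.
# It uses a shared small weapon-reload attenuation asset and the first-person ranged-weapon fire sound class.
# A concurrency override prevents new instances once the limit is reached.
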SoundCue /Figment_Martin/Athena/Sounds/LTMs/Ashton/Milo/Ashton_Milo_Pickup_Cue
root=WavePlayer_1
nodes [{id=WavePlayer_1, properties {soundWaveAssetPtr=&/Ashton_Milo_BR_Pickup}}]
properties {
 attenuationSettings=/Figment_Martin/Sounds/Attenuation/Fort_Weapon_Reload_Sml_Att
 bOverrideConcurrency=true
 maxDistance=2100
 soundClassObject=&Fort_WeaponRanged_Fire_1P_PSM
 volumeMultiplier=0.5915
 attenuationOverrides {distanceAlgorithm=NaturalSound, falloffDistance=4000, attenuationShapeExtents {x=1000}}
 concurrencyOverrides {resolutionRule=PreventNew}
}
