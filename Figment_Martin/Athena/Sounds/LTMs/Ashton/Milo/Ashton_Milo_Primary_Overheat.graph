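# Sound cue for the overheat sound of the Milo primary weapon in the Ashton LTM.
# A mixer layers a reused mounted turret overheat wave, pitched up by a randomized modulator and mixed at low volume, with the Milo BR drop wave at full level.
# It uses a small weapon reload attenuation asset, a first-person ranged fire sound class, boosted volume, and a concurrency override that prevents new instances.
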SoundCue /Figment_Martin/Athena/Sounds/LTMs/Ashton/Milo/Ashton_Milo_Primary_Overheat
root=Mixer_0
nodes [
 {id=Mixer_0, children [Modulator_0,WavePlayer_1], properties {inputVolume [0.25,1]}}
 {id=Modulator_0, children [WavePlayer_2], properties {pitchMax=1.5, pitchMin=1.4}}
 {id=WavePlayer_1, properties {soundWaveAssetPtr=/Game/Athena/Sounds/LTMs/Ashton/Milo/Ashton_Milo_BR_Drop}}
 {id=WavePlayer_2, properties {soundWaveAssetPtr=/Game/Sounds/Fort_Traps/MountedTurret/Turret_Ballistic_Overheat_02}}
]
properties {
 attenuationSettings=/Figment_Martin/Sounds/Attenuation/Fort_Weapon_Reload_Sml_Att
 bOverrideConcurrency=true
 maxDistance=2100
 soundClassObject=&Fort_WeaponRanged_Fire_1P_PSM
 volumeMultiplier=1.69
 attenuationOverrides {distanceAlgorithm=NaturalSound, falloffDistance=4000, attenuationShapeExtents {x=1000}}
 concurrencyOverrides {resolutionRule=PreventNew}
}
