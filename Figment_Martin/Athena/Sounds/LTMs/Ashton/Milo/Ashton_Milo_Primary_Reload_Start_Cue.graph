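# Sound cue for the start of the reload on Milo's primary weapon in the Ashton limited-time mode content.
# A single wave player feeds the reload-start wave straight to the output with a slight volume boost.
# It uses the shared small weapon-reload attenuation asset and routes to a first-person ranged weapon fire sound class.
# Overridden concurrency prevents new instances from starting once the limit is reached.
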SoundCue /Figment_Martin/Athena/Sounds/LTMs/Ashton/Milo/Ashton_Milo_Primary_Reload_Start_Cue
root=WavePlayer_0
nodes [{id=WavePlayer_0, properties {soundWaveAssetPtr=&/Ashton_Milo_BR_Reload_Start}}]
properties {
 attenuationSettings=/Figment_Martin/Sounds/Attenuation/Fort_Weapon_Reload_Sml_Att
 bOverrideConcurrency=true
 maxDistance=2100
 soundClassObject=&Fort_WeaponRanged_Fire_1P_PSM
 volumeMultiplier=1.183
 attenuationOverrides {distanceAlgorithm=NaturalSound, falloffDistance=4000, attenuationShapeExtents {x=1000}}
 concurrencyOverrides {resolutionRule=PreventNew}
}
