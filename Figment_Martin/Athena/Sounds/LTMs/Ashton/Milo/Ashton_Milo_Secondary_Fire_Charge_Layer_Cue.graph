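# Sound cue for the charge layer of the Ashton Milo secondary fire weapon sound.
# A single wave player plays the amr_superfire rifle wave with a slight volume boost, routed to the first-person ranged weapon fire sound class.
# It references the shared Ashton ability attenuation asset, also carries natural-sound falloff values, and overrides concurrency to block new instances rather than stopping playing ones.
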SoundCue /Figment_Martin/Athena/Sounds/LTMs/Ashton/Milo/Ashton_Milo_Secondary_Fire_Charge_Layer_Cue
root=WavePlayer_12
nodes [{id=WavePlayer_12, properties {soundWaveAssetPtr=/Game/Sounds/Fort_Weapons/Rifle/amr_superfire}}]
properties {
 attenuationSettings=/Game/Sounds/Attenuation/Fort_Ability_Ashton_Gen
 bOverrideConcurrency=true
 maxDistance=9300
 soundClassObject=&Fort_WeaponRanged_Fire_1P_PSM
 volumeMultiplier=1.183
 attenuationOverrides {distanceAlgorithm=NaturalSound, falloffDistance=4000, attenuationShapeExtents {x=1000}}
 concurrencyOverrides {resolutionRule=PreventNew}
}
