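# Sound class for distant weapon projectile sounds, a child of the Fort_WeaponProjectile class.
# It boosts volume, bleeds part of the signal to the LFE channel and enables effects processing.
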SoundClass /Figment_Martin/Sounds/FortSoundClassesAndModes/Fort_Weapon_Projectile_Distant
parent=&/Fort_WeaponProjectile
properties {properties {bApplyEffects=true, lFEBleed=0.25, volume=1.6}}
