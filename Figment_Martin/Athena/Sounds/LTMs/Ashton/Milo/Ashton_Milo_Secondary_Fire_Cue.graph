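# Sound cue for the secondary fire of the Ashton Milo LTM weapon, built from launcher fire recordings.
# Nested mixers layer close, medium and distant branches, with the latter two reduced in level.
# Each branch picks randomly among four waves, then passes through a modulator, its own attenuation and a sound class override.
# The distant layer is pitched down, medium shares the Distant sound class name, and concurrency is overridden to prevent new instances.
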SoundCue /Figment_Martin/Athena/Sounds/LTMs/Ashton/Milo/Ashton_Milo_Secondary_Fire_Cue
root=Mixer_4
nodes [
 {id=Attenuation_6, children [Modulator_6], properties {attenuationSettings=&Fort_Weapon_Projectile_Close_Athena_Att}}
 {
  id=Attenuation_7
  children [Modulator_4]
  properties {attenuationSettings=/Figment_Martin/Sounds/Attenuation/Fort_Weapon_Projectile_Distant_Athena_Att}
 }
 {
  id=Attenuation_8
  children [Modulator_7]
  properties {attenuationSettings=/Figment_Martin/Sounds/Attenuation/Fort_Weapon_Projectile_Medium_Athena_Att}
 }
 {id=Mixer_4, children [Mixer_5,SoundClass_8], properties {inputVolume [1,0.5]}}
 {id=Mixer_5, children [SoundClass_7,SoundClass_9], properties {inputVolume [1,0.7]}}
 {id=Modulator_4, children [Random_2], properties {pitchMax=0.8, pitchMin=0.7, volumeMax=1, volumeMin=1}}
 {id=Modulator_6, children [Random_0]}
 {id=Modulator_7, children [Random_1], properties {volumeMax=1, volumeMin=1}}
 {id=Random_0, children [WavePlayer_0,WavePlayer_1,WavePlayer_2,WavePlayer_3], properties {weights [1,1,1,1]}}
 {id=Random_1, children [WavePlayer_4,WavePlayer_5,WavePlayer_6,WavePlayer_7], properties {weights [1,1,1,1]}}
 {id=Random_2, children [WavePlayer_8, WavePlayer_9, WavePlayer_10, WavePlayer_11], properties {weights [1,1,1,1]}}
 {
  id=SoundClass_7
  children [Attenuation_6]
  properties {soundClassOverride=/Figment_Martin/Sounds/FortSoundClassesAndModes/Fort_Weapon_Projectile_PSM}
 }
 {
  id=SoundClass_8
  children [Attenuation_7]
  properties {soundClassOverride=/Figment_Martin/Sounds/FortSoundClassesAndModes/Fort_Weapon_Projectile_Distant}
 }
 {
  id=SoundClass_9
  children [Attenuation_8]
  properties {soundClassOverride=&Fort_Weapon_Projectile_Distant}
 }
 {id=WavePlayer_0, properties {soundWaveAssetPtr=&/Ashton_Milo_Launcher_Fire_Close_01}}
 {id=WavePlayer_1, properties {soundWaveAssetPtr=&/Ashton_Milo_Launcher_Fire_Close_02}}
 {id=WavePlayer_10, properties {soundWaveAssetPtr=&/Ashton_Milo_Launcher_Fire_Dist_03}}
 {id=WavePlayer_11, properties {soundWaveAssetPtr=&/Ashton_Milo_Launcher_Fire_Dist_04}}
 {id=WavePlayer_2, properties {soundWaveAssetPtr=&/Ashton_Milo_Launcher_Fire_Close_03}}
 {id=WavePlayer_3, properties {soundWaveAssetPtr=&/Ashton_Milo_Launcher_Fire_Close_04}}
 {id=WavePlayer_4, properties {soundWaveAssetPtr=&/Ashton_Milo_Launcher_Fire_Med_01}}
 {id=WavePlayer_5, properties {soundWaveAssetPtr=&/Ashton_Milo_Launcher_Fire_Med_02}}
 {id=WavePlayer_6, properties {soundWaveAssetPtr=&/Ashton_Milo_Launcher_Fire_Med_03}}
 {id=WavePlayer_7, properties {soundWaveAssetPtr=&/Ashton_Milo_Launcher_Fire_Med_04}}
 {id=WavePlayer_8, properties {soundWaveAssetPtr=&/Ashton_Milo_Launcher_Fire_Dist_01}}
 {id=WavePlayer_9, properties {soundWaveAssetPtr=&/Ashton_Milo_Launcher_Fire_Dist_02}}
]
properties {
 bOverrideConcurrency=true
 maxDistance=18000
 soundClassObject=&Fort_WeaponRanged_Fire_1P_PSM
 volumeMultiplier=2.028
 attenuationOverrides {distanceAlgorithm=NaturalSound, falloffDistance=4000, attenuationShapeExtents {x=1000}}
 concurrencyOverrides {resolutionRule=PreventNew}
}
