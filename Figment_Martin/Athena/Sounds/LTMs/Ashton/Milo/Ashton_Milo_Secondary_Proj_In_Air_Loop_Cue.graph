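# Sound cue for the looping in-air sound of the Ashton Milo secondary projectile, reusing the launcher charge loop wave.
# An enveloper fades the loop's volume in, then a modulator randomly raises its pitch.
# Distant-rocket attenuation with overrides and a distant-projectile sound class override wrap the chain.
# Cue-level overrides set natural-sound falloff and an owner-limited concurrency cap that stops the quietest voice.
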SoundCue /Figment_Martin/Athena/Sounds/LTMs/Ashton/Milo/Ashton_Milo_Secondary_Proj_In_Air_Loop_Cue
root=SoundClass_6
nodes [
 {
  id=Attenuation_11
  children [Modulator_7]
  properties {
   attenuationSettings=/Game/Sounds/Attenuation/Fort_Weapon_Rocket_Distant_Athena_Att
   attenuationOverrides {falloffDistance=2600, attenuationShapeExtents {x=1000}}
  }
 }
 {
  id=Enveloper_4
  children [WavePlayer_3]
  properties {
   pitchCurve {editorCurveData {keys [{time=0}]}}
   volumeCurve {editorCurveData {keys [{interpMode=RCIM_Cubic, tangentMode=RCTM_User, time=0, value=0}, {interpMode=RCIM_Cubic, tangentMode=RCTM_User, time=1, value=1}]}}
  }
 }
 {id=Modulator_7, children [Enveloper_4], properties {pitchMax=1.3, pitchMin=1.2}}
 {
  id=SoundClass_6
  children [Attenuation_11]
  properties {soundClassOverride=/Figment_Martin/Sounds/FortSoundClassesAndModes/Fort_Weapon_Projectile_Distant}
 }
 {
  id=WavePlayer_3
  properties {bLooping=true, soundWaveAssetPtr=&/Ashton_Milo_Launcher_Charge_Loop_01}
 }
]
properties {
 bOverrideAttenuation=true
 bOverrideConcurrency=true
 maxDistance=15000
 soundClassObject=&Fort_WeaponRanged_Fire_Distant
 volumeMultiplier=1.56
 attenuationOverrides {distanceAlgorithm=NaturalSound}
 concurrencyOverrides {bLimitToOwner=true, maxCount=3, resolutionRule=StopQuietest}
}
