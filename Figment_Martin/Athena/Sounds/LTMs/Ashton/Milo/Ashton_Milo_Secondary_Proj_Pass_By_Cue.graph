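# Sound cue for the pass-by of the Ashton Milo secondary projectile.
# A Random node picks one of four equally weighted launcher projectile travel waves, then feeds it through a Modulator left at default settings.
# The cue boosts volume, uses the Ashton general ability attenuation asset, and is assigned to a first-person ranged weapon fire sound class despite being a projectile sound.
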SoundCue /Figment_Martin/Athena/Sounds/LTMs/Ashton/Milo/Ashton_Milo_Secondary_Proj_Pass_By_Cue
root=Modulator_0
nodes [
 {id=Modulator_0, children [Random_0]}
 {id=Random_0, children [WavePlayer_1,WavePlayer_2,WavePlayer_3,WavePlayer_4], properties {weights [1,1,1,1]}}
 {id=WavePlayer_1, properties {soundWaveAssetPtr=&/Ashton_Milo_Launcher_Proj_Travel_01}}
 {id=WavePlayer_2, properties {soundWaveAssetPtr=&/Ashton_Milo_Launcher_Proj_Travel_02}}
 {id=WavePlayer_3, properties {soundWaveAssetPtr=&/Ashton_Milo_Launcher_Proj_Travel_03}}
 {id=WavePlayer_4, properties {soundWaveAssetPtr=&/Ashton_Milo_Launcher_Proj_Travel_04}}
]
properties {
 attenuationSettings=/Game/Sounds/Attenuation/Fort_Ability_Ashton_Gen
 maxDistance=9300
 soundClassObject=&Fort_WeaponRanged_Fire_1P_PSM
 volumeMultiplier=1.69
}
