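# Sound cue for the secondary explosion of the Ashton Milo launcher projectile.
# After a short random delay, it mixes close, medium and distant layers, each with its own modulator, attenuation and sound class override.
# Layers pick random waves: close uses Milo alone, medium adds quieter DrunkenRocket, and distant adds RatRod.
# Overridden concurrency caps simultaneous instances at four, stopping the oldest.
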SoundCue /Figment_Martin/Athena/Sounds/LTMs/Ashton/Milo/SC_Ashton_Milo_Secondary_Expl
root=Delay_1
nodes [
 {
  id=Attenuation_0
  children [Mixer_3]
  properties {attenuationSettings=/Game/Sounds/Attenuation/Fort_Weapon_Milo_Medium_Explosion_Athena_Att, attenuationOverrides {falloffDistance=1600}}
 }
 {
  id=Attenuation_4
  children [Random_8]
  properties {attenuationSettings=/Game/Sounds/Attenuation/Fort_Weapon_Milo_Close_Explosion_Athena_Att, attenuationOverrides {falloffDistance=1600}}
 }
 {
  id=Attenuation_5
  children [Modulator_5]
  properties {
   attenuationSettings=/Game/Sounds/Attenuation/Fort_Weapon_Milo_Distant_Explosion_Athena_Att
   attenuationOverrides {falloffDistance=3000, attenuationShapeExtents {x=1000}}
  }
 }
 {id=Delay_1, children [Mixer_0], properties {delayMax=0.1, delayMin=0.05}}
 {id=Mixer_0, children [SoundClass_3,SoundClass_0,SoundClass_2], properties {inputVolume [1,1,1]}}
 {id=Mixer_2, children [Random_3,Random_5], properties {inputVolume [1,1]}}
 {id=Mixer_3, children [Random_4,Random_7], properties {inputVolume [1,0.5]}}
 {id=Modulator_0, children [Attenuation_4]}
 {id=Modulator_1, children [Attenuation_0]}
 {id=Modulator_5, children [Mixer_2], properties {pitchMax=1.1, pitchMin=0.9}}
 {id=Random_3, children [WavePlayer_12, WavePlayer_13, WavePlayer_14], properties {weights [1,1,1]}}
 {id=Random_4, children [WavePlayer_20, WavePlayer_21, WavePlayer_22], properties {weights [1,1,1]}}
 {id=Random_5, children [WavePlayer_16, WavePlayer_17, WavePlayer_18], properties {weights [1,1,1]}}
 {id=Random_7, children [WavePlayer_27, WavePlayer_28, WavePlayer_29, WavePlayer_30], properties {weights [1,1,1,1]}}
 {id=Random_8, children [WavePlayer_31, WavePlayer_32, WavePlayer_33, WavePlayer_34], properties {weights [1,1,1,1]}}
 {
  id=SoundClass_0
  children [Modulator_1]
  properties {soundClassOverride=/Figment_Martin/Sounds/FortSoundClassesAndModes/Fort_Weapon_Projectile_PSM}
 }
 {
  id=SoundClass_2
  children [Attenuation_5]
  properties {soundClassOverride=/Figment_Martin/Sounds/FortSoundClassesAndModes/Fort_Weapon_Projectile_Distant}
 }
 {
  id=SoundClass_3
  children [Modulator_0]
  properties {soundClassOverride=&Fort_Weapon_Projectile_PSM}
 }
 {id=WavePlayer_12, properties {soundWaveAssetPtr=&/Ashton_Milo_Launcher_Proj_Expl_Dist_01}}
 {id=WavePlayer_13, properties {soundWaveAssetPtr=&/Ashton_Milo_Launcher_Proj_Expl_Dist_02}}
 {id=WavePlayer_14, properties {soundWaveAssetPtr=&/Ashton_Milo_Launcher_Proj_Expl_Dist_03}}
 {id=WavePlayer_16, properties {soundWaveAssetPtr=&/RatRod_Launcher_Explode_Distant_03}}
 {id=WavePlayer_17, properties {soundWaveAssetPtr=&/RatRod_Launcher_Explode_Distant_02}}
 {id=WavePlayer_18, properties {soundWaveAssetPtr=&/RatRod_Launcher_Explode_Distant_01}}
 {id=WavePlayer_20, properties {soundWaveAssetPtr=&/Ashton_Milo_Launcher_Proj_Expl_Med_02}}
 {id=WavePlayer_21, properties {soundWaveAssetPtr=&/Ashton_Milo_Launcher_Proj_Expl_Med_03}}
 {id=WavePlayer_22, properties {soundWaveAssetPtr=&/Ashton_Milo_Launcher_Proj_Expl_Med_04}}
 {id=WavePlayer_27, properties {soundWaveAssetPtr=&/DrunkenRocket_Explosion_Medium_01}}
 {id=WavePlayer_28, properties {soundWaveAssetPtr=&/DrunkenRocket_Explosion_Medium_02}}
 {id=WavePlayer_29, properties {soundWaveAssetPtr=&/DrunkenRocket_Explosion_Medium_03}}
 {id=WavePlayer_30, properties {soundWaveAssetPtr=&/DrunkenRocket_Explosion_Medium_04}}
 {id=WavePlayer_31, properties {soundWaveAssetPtr=&/Ashton_Milo_Launcher_Proj_Expl_Close_V2_01}}
 {id=WavePlayer_32, properties {soundWaveAssetPtr=&/Ashton_Milo_Launcher_Proj_Expl_Close_V2_02}}
 {id=WavePlayer_33, properties {soundWaveAssetPtr=&/Ashton_Milo_Launcher_Proj_Expl_Close_V2_03}}
 {id=WavePlayer_34, properties {soundWaveAssetPtr=&/Ashton_Milo_Launcher_Proj_Expl_Close_V2_04}}
]
properties {
 bOverrideConcurrency=true
 maxDistance=15000
 soundClassObject=&Fort_WeaponRanged_Fire_1P_PSM
 volumeMultiplier=2.5688
 concurrencyOverrides {maxCount=4, resolutionRule=StopOldest, volumeScale=0.5}
}
